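# Secondary-motion AnimBP for the male medium padded armor, following the attached parent mesh via Copy Pose From Mesh.
# Weighted constraints position flap, knee, pouch and thigh-ring bones; Anim Dynamics springs add jiggle, with Look At aiming thigh rings.
# Additive bone offsets fix collar, back-tube, clavicle and right thigh, weighted by EventGraph head/thigh rotations and targeting events.
# An earlier duplicate offset chain is unconnected to the output; head_rot_right drives nothing.
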AnimBlueprint /Game/Characters/Player/Male/Medium/Bodies/M_MED_Padded_Armor/Meshes/M_MED_Padded_Armor_AnimBP
target_skeleton=&/M_MED_Padded_Armor_Skeleton
graphs [
 AnimGraph {nodes [
   n1 = "Transform (Modify) Bone - Bone: backTubeAttach_c"(Alpha=Get head_rot_back() -> (head_rot_back:double), Translation="-9.5,-4.5,0", Rotation="0,0,0", Scale="1,1,1") {
    rotationMode=BMM_Additive
    rotationSpace=BCS_BoneSpace
    scaleMode=BMM_Additive
    scaleSpace=BCS_BoneSpace
    translationMode=BMM_Additive
    translationSpace=BCS_ParentBoneSpace
    alphaBoolBlend {blendOption=Linear}
    alphaScaleBiasClamp {bInterpResult=true, interpSpeedDecreasing=25, interpSpeedIncreasing=25}
    boneToModify {boneName=backTubeAttach_c}
    translation=<-9.5 -4.5>
   }
   n2 = "Transform (Modify) Bone - Bone: head_collar_squish_r"(ComponentPose=@n1, Alpha=1, Translation="-3,-2,5", Rotation="0,0,0", Scale="1,1,1") {
    rotationMode=BMM_Additive
    rotationSpace=BCS_BoneSpace
    scaleMode=BMM_Additive
    scaleSpace=BCS_BoneSpace
    translationMode=BMM_Additive
    translationSpace=BCS_ParentBoneSpace
    alphaBoolBlend {blendOption=Linear}
    alphaScaleBiasClamp {interpSpeedDecreasing=25, interpSpeedIncreasing=25}
    boneToModify {boneName=head_collar_squish_r}
    translation=<-3 -2 5>
   }
   n3 = "Transform (Modify) Bone - Bone: head_collar_squish_l"(ComponentPose=@n2, Alpha=1, Translation="-3,-2,-5", Rotation="0,0,0", Scale="1,1,1") {
    rotationMode=BMM_Additive
    rotationSpace=BCS_BoneSpace
    scaleMode=BMM_Additive
    scaleSpace=BCS_BoneSpace
    translationMode=BMM_Additive
    translationSpace=BCS_ParentBoneSpace
    alphaBoolBlend {blendOption=Linear}
    alphaScaleBiasClamp {interpSpeedDecreasing=25, interpSpeedIncreasing=25}
    boneToModify {boneName=head_collar_squish_l}
    translation=<-3 -2 -5>
   }
   n4 = "Transform (Modify) Bone - Bone: neckCollar_l"(ComponentPose=@n3, Alpha=Get head_rot_left() -> (head_rot_left:double), Translation="0,0,-5", Rotation="0,0,0", Scale="1,1,1") {
    rotationMode=BMM_Additive
    rotationSpace=BCS_BoneSpace
    scaleMode=BMM_Additive
    scaleSpace=BCS_BoneSpace
    translationMode=BMM_Additive
    translationSpace=BCS_ParentBoneSpace
    alphaBoolBlend {blendOption=Linear}
    alphaScaleBiasClamp {interpSpeedDecreasing=25, interpSpeedIncreasing=25}
    boneToModify {boneName=neckCollar_l}
    translation {z=-5}
   }
   n5 = Get ads_rotate() -> (ads_rotate:double)
   n6 = "Transform (Modify) Bone - Bone: neckCollar_r"(ComponentPose=@n4, Alpha=@n5, Translation="0,0,3.5", Rotation="0,0,0", Scale="1,1,1") {
    rotationMode=BMM_Additive
    rotationSpace=BCS_BoneSpace
    scaleMode=BMM_Additive
    scaleSpace=BCS_BoneSpace
    translationMode=BMM_Additive
    translationSpace=BCS_ParentBoneSpace
    alphaBoolBlend {blendOption=Linear}
    alphaScaleBiasClamp {bInterpResult=true, interpSpeedDecreasing=25, interpSpeedIncreasing=25}
    boneToModify {boneName=neckCollar_r}
    translation {z=3.5}
   }
   "Transform (Modify) Bone - Bone: clavicle_ads_r"(ComponentPose=@n6, Alpha=@n5, Translation="-3.5,0,0", Rotation="0,0,0", Scale="1,1,1") {
    rotationMode=BMM_Additive
    rotationSpace=BCS_BoneSpace
    scaleMode=BMM_Additive
    scaleSpace=BCS_BoneSpace
    translationMode=BMM_Additive
    translationSpace=BCS_ParentBoneSpace
    alphaBoolBlend {blendOption=Linear}
    alphaScaleBiasClamp {bInterpResult=true, interpSpeedDecreasing=25, interpSpeedIncreasing=25}
    boneToModify {boneName=clavicle_ads_r}
    translation {x=-3.5}
   }
   n7 = "Constraint - bicepFlap_fnt_r_zero (PPPPPPP)"!(
    ComponentPose=Local To Component(LocalPose=Copy Pose From Mesh() {bUseAttachedParent=true})
    Alpha=1
    ConstraintWeights_0=0.1982
    ConstraintWeights_1:float=0.0625
    ConstraintWeights_2:float=0.0111
    ConstraintWeights_3:float=0.1295
    ConstraintWeights_4:float=0.1579
    ConstraintWeights_5:float=0.4199
    ConstraintWeights_6:float=0.0208
   ) -> (Pose) {
    alphaBoolBlend {blendOption=Linear}
    boneToModify {boneName=bicepFlap_fnt_r_zero}
    constraintSetup [
     {targetBone {boneName=upperarm_r}}
     {targetBone {boneName=bicepFlap_bck_r}}
     {targetBone {boneName=lowerarm_r}}
     {targetBone {boneName=upperarm_twist_01_r}}
     {targetBone {boneName=upperarm_twist_02_r}}
     {targetBone {boneName=bicepFlap_fnt_r}}
     {targetBone {boneName=clavicle_r}}
    ]
    constraintWeights [1,1,1,1,1,1,1]
   }
   n8 = "Constraint - bicepFlap_bck_r_zero (PPPPPP)"!(
    ComponentPose=@n7
    Alpha=1
    ConstraintWeights_0=0.2717
    ConstraintWeights_1:float=0.2989
    ConstraintWeights_2:float=0.014
    ConstraintWeights_3:float=0.1737
    ConstraintWeights_4:float=0.2142
    ConstraintWeights_5:float=0.0273
   ) -> (Pose) {
    alphaBoolBlend {blendOption=Linear}
    boneToModify {boneName=bicepFlap_bck_r_zero}
    constraintSetup [
     {targetBone {boneName=upperarm_r}}
     {targetBone {boneName=bicepFlap_bck_r}}
     {targetBone {boneName=lowerarm_r}}
     {targetBone {boneName=upperarm_twist_01_r}}
     {targetBone {boneName=upperarm_twist_02_r}}
     {targetBone {boneName=clavicle_r}}
    ]
    constraintWeights [1,1,1,1,1,1]
   }
   n9 = "Constraint - bicepFlap_fnt_l_zero (PPPPP)"!(
    ComponentPose=@n8
    Alpha=1
    ConstraintWeights_0=0.3358
    ConstraintWeights_1:float=0.0433
    ConstraintWeights_2:float=0.2656
    ConstraintWeights_3:float=0.0358
    ConstraintWeights_4:float=0.3195
   ) -> (Pose) {
    alphaBoolBlend {blendOption=Linear}
    boneToModify {boneName=bicepFlap_fnt_l_zero}
    constraintSetup [
     {targetBone {boneName=upperarm_twist_02_l}}
     {targetBone {boneName=bicepFlap_fnt_l}}
     {targetBone {boneName=upperarm_twist_01_l}}
     {targetBone {boneName=clavicle_l}}
     {targetBone {boneName=upperarm_l}}
    ]
    constraintWeights [1,1,1,1,1]
   }
   n10 = "Constraint - bicepFlap_bck_l_zero (PPPPP)"!(
    ComponentPose=@n9
    Alpha=1
    ConstraintWeights_0=0.3441
    ConstraintWeights_1:float=0.0365
    ConstraintWeights_2:float=0.2531
    ConstraintWeights_3:float=0.0302
    ConstraintWeights_4:float=0.3361
   ) -> (Pose) {
    alphaBoolBlend {blendOption=Linear}
    boneToModify {boneName=bicepFlap_bck_l_zero}
    constraintSetup [
     {targetBone {boneName=upperarm_twist_02_l}}
     {targetBone {boneName=bicepFlap_bck_l}}
     {targetBone {boneName=upperarm_twist_01_l}}
     {targetBone {boneName=clavicle_l}}
     {targetBone {boneName=upperarm_l}}
    ]
    constraintWeights [1,1,1,1,1]
   }
   n11 = Anim Dynamics - Bone: bicepFlap_bck_l(ComponentPose=@n10, Alpha=0.5, ExternalForce="0,0,0") {
    angularDampingOverride=1
    angularSpringConstant=1500
    bAngularSpring=true
    bGravityOverrideInSimSpace=true
    bOverrideAngularDamping=true
    gravityScale=0.35
    simulationSpace=RootRelative
    alphaBoolBlend {blendOption=Linear}
    boundBone {boneName=bicepFlap_bck_l}
    physicsBodyDefinitions [{
      sphereCollisionRadius=0
      boundBone {boneName=bicepFlap_bck_l}
      boxExtents=<10 20 20>
      constraintSetup {angularLimitsMax=<0 0 0>, angularLimitsMin=<0 0 -40>, angularTarget=<0 0 0>}
      localJointOffset=<10 0 0>
     }]
   }
   n12 = Anim Dynamics - Bone: bicepFlap_bck_r(ComponentPose=@n11, Alpha=0.5, ExternalForce="0,0,0") {
    angularDampingOverride=1
    angularSpringConstant=1500
    bAngularSpring=true
    bGravityOverrideInSimSpace=true
    bOverrideAngularDamping=true
    gravityScale=0.35
    simulationSpace=RootRelative
    alphaBoolBlend {blendOption=Linear}
    boundBone {boneName=bicepFlap_bck_r}
    physicsBodyDefinitions [{
      sphereCollisionRadius=0
      boundBone {boneName=bicepFlap_bck_r}
      boxExtents=<10 20 20>
      constraintSetup {angularLimitsMax=<0 0 0>, angularLimitsMin=<0 0 -40>, angularTarget=<0 0 0>}
      localJointOffset=<10 0 0>
     }]
   }
   n13 = Anim Dynamics - Bone: bicepFlap_fnt_l(ComponentPose=@n12, Alpha=0.5, ExternalForce="0,0,0") {
    angularDampingOverride=1
    angularSpringConstant=1500
    bAngularSpring=true
    bGravityOverrideInSimSpace=true
    bOverrideAngularDamping=true
    gravityScale=0.35
    simulationSpace=RootRelative
    alphaBoolBlend {blendOption=Linear}
    boundBone {boneName=bicepFlap_fnt_l}
    physicsBodyDefinitions [{
      sphereCollisionRadius=0
      boundBone {boneName=bicepFlap_fnt_l}
      boxExtents=<10 20 20>
      constraintSetup {angularLimitsMax=<0 0 0>, angularLimitsMin=<0 0 -40>, angularTarget=<0 0 0>}
      localJointOffset=<10 0 0>
     }]
   }
   n14 = Anim Dynamics - Bone: bicepFlap_fnt_r(ComponentPose=@n13, Alpha=0.5, ExternalForce="0,0,0") {
    angularDampingOverride=1
    angularSpringConstant=1500
    bAngularSpring=true
    bGravityOverrideInSimSpace=true
    bOverrideAngularDamping=true
    gravityScale=0.35
    simulationSpace=RootRelative
    alphaBoolBlend {blendOption=Linear}
    boundBone {boneName=bicepFlap_fnt_r}
    physicsBodyDefinitions [{
      sphereCollisionRadius=0
      boundBone {boneName=bicepFlap_fnt_r}
      boxExtents=<10 20 20>
      constraintSetup {angularLimitsMax=<0 0 0>, angularLimitsMin=<0 0 -40>, angularTarget=<0 0 0>}
      localJointOffset=<10 0 0>
     }]
   }
   n15 = "Constraint - knee_l (PP)"!(ComponentPose=@n14, Alpha=1, ConstraintWeights_0=0.6, ConstraintWeights_1:float=0.4) -> (Pose) {
    alphaBoolBlend {blendOption=Linear}
    boneToModify {boneName=knee_l}
    constraintSetup [{targetBone {boneName=thigh_l}}, {targetBone {boneName=calf_l}}]
    constraintWeights [1,1]
   }
   n16 = "Constraint - knee_r (PP)"!(ComponentPose=@n15, Alpha=1, ConstraintWeights_0=0.6, ConstraintWeights_1:float=0.4) -> (Pose) {
    alphaBoolBlend {blendOption=Linear}
    boneToModify {boneName=knee_r}
    constraintSetup [{targetBone {boneName=thigh_r}}, {targetBone {boneName=calf_r}}]
    constraintWeights [1,1]
   }
   n17 = Anim Dynamics - Bone: ankleFlap_l(ComponentPose=@n16, Alpha=1, ExternalForce="0,0,0") {
    angularDampingOverride=1
    angularSpringConstant=1500
    bAngularSpring=true
    bGravityOverrideInSimSpace=true
    bOverrideAngularDamping=true
    gravityScale=0.35
    simulationSpace=RootRelative
    alphaBoolBlend {blendOption=Linear}
    boundBone {boneName=ankleFlap_l}
    physicsBodyDefinitions [{
      sphereCollisionRadius=0
      boundBone {boneName=ankleFlap_l}
      boxExtents=<10 20 20>
      constraintSetup {angularLimitsMax=<0 0 0>, angularLimitsMin=<0 0 -40>, angularTarget=<0 0 0>}
      localJointOffset=<10 0 0>
     }]
   }
   n18 = Anim Dynamics - Bone: ankleFlap_r(ComponentPose=@n17, Alpha=1, ExternalForce="0,0,0") {
    angularDampingOverride=1
    angularSpringConstant=1500
    bAngularSpring=true
    bGravityOverrideInSimSpace=true
    bOverrideAngularDamping=true
    gravityScale=0.35
    simulationSpace=RootRelative
    alphaBoolBlend {blendOption=Linear}
    boundBone {boneName=ankleFlap_r}
    physicsBodyDefinitions [{
      sphereCollisionRadius=0
      boundBone {boneName=ankleFlap_r}
      boxExtents=<10 20 20>
      constraintSetup {angularLimitsMax=<0 0 0>, angularLimitsMin=<0 0 -40>, angularTarget=<0 0 0>}
      localJointOffset=<10 0 0>
     }]
   }
   n19 = "Constraint - calfFlap_2_r_zero (PPP)"!(ComponentPose=@n18, Alpha=1, ConstraintWeights_0=0.1951, ConstraintWeights_1:float=0.016, ConstraintWeights_2:float=0.7889) -> (Pose) {
    alphaBoolBlend {blendOption=Linear}
    boneToModify {boneName=calfFlap_2_r_zero}
    constraintSetup [{targetBone {boneName=thigh_r}}, {targetBone {boneName=calf_twist_02_r}}, {targetBone {boneName=calf_r}}]
    constraintWeights [1,1,1]
   }
   n20 = "Constraint - calfFlap_1_r_zero (PP)"!(ComponentPose=@n19, Alpha=1, ConstraintWeights_0=0.6117, ConstraintWeights_1:float=0.3883) -> (Pose) {
    alphaBoolBlend {blendOption=Linear}
    boneToModify {boneName=calfFlap_1_r_zero}
    constraintSetup [{targetBone {boneName=thigh_r}}, {targetBone {boneName=calf_r}}]
    constraintWeights [1,1]
   }
   n21 = "Constraint - calfFlap_1_l_zero (PP)"!(ComponentPose=@n20, Alpha=1, ConstraintWeights_0=0.3883, ConstraintWeights_1:float=0.6117) -> (Pose) {
    alphaBoolBlend {blendOption=Linear}
    boneToModify {boneName=calfFlap_1_l_zero}
    constraintSetup [{targetBone {boneName=calf_l}}, {targetBone {boneName=thigh_l}}]
    constraintWeights [1,1]
   }
   n22 = "Constraint - calfFlap_2_l_zero (PPP)"!(ComponentPose=@n21, Alpha=1, ConstraintWeights_0=0.7889, ConstraintWeights_1:float=0.016, ConstraintWeights_2:float=0.1951) -> (Pose) {
    alphaBoolBlend {blendOption=Linear}
    boneToModify {boneName=calfFlap_2_l_zero}
    constraintSetup [{targetBone {boneName=calf_l}}, {targetBone {boneName=calf_twist_02_l}}, {targetBone {boneName=thigh_l}}]
    constraintWeights [1,1,1]
   }
   n23 = Anim Dynamics - Bone: calfFlap_1_r(ComponentPose=@n22, Alpha=1, ExternalForce="0,0,0") {
    angularDampingOverride=1
    angularSpringConstant=1500
    bAngularSpring=true
    bGravityOverrideInSimSpace=true
    bOverrideAngularDamping=true
    gravityScale=0.35
    simulationSpace=RootRelative
    alphaBoolBlend {blendOption=Linear}
    boundBone {boneName=calfFlap_1_r}
    physicsBodyDefinitions [{
      sphereCollisionRadius=0
      boundBone {boneName=calfFlap_1_r}
      boxExtents=<10 20 20>
      constraintSetup {angularLimitsMax=<0 0 0>, angularLimitsMin=<0 0 -40>, angularTarget=<0 0 0>}
      localJointOffset=<10 0 0>
     }]
   }
   n24 = Anim Dynamics - Bone: calfFlap_2_r(ComponentPose=@n23, Alpha=1, ExternalForce="0,0,0") {
    angularDampingOverride=1
    angularSpringConstant=1500
    bAngularSpring=true
    bGravityOverrideInSimSpace=true
    bOverrideAngularDamping=true
    gravityScale=0.35
    simulationSpace=RootRelative
    alphaBoolBlend {blendOption=Linear}
    boundBone {boneName=calfFlap_2_r}
    physicsBodyDefinitions [{
      sphereCollisionRadius=0
      boundBone {boneName=calfFlap_2_r}
      boxExtents=<10 20 20>
      constraintSetup {angularLimitsMax=<0 0 0>, angularLimitsMin=<0 0 -40>, angularTarget=<0 0 0>}
      localJointOffset=<10 0 0>
     }]
   }
   n25 = Anim Dynamics - Bone: calfFlap_1_l(ComponentPose=@n24, Alpha=1, ExternalForce="0,0,0") {
    angularDampingOverride=1
    angularSpringConstant=1500
    bAngularSpring=true
    bGravityOverrideInSimSpace=true
    bOverrideAngularDamping=true
    gravityScale=0.35
    simulationSpace=RootRelative
    alphaBoolBlend {blendOption=Linear}
    boundBone {boneName=calfFlap_1_l}
    physicsBodyDefinitions [{
      sphereCollisionRadius=0
      boundBone {boneName=calfFlap_1_l}
      boxExtents=<10 20 20>
      constraintSetup {angularLimitsMax=<0 0 0>, angularLimitsMin=<0 0 -40>, angularTarget=<0 0 0>}
      localJointOffset=<10 0 0>
     }]
   }
   n26 = Anim Dynamics - Bone: calfFlap_2_l(ComponentPose=@n25, Alpha=1, ExternalForce="0,0,0") {
    angularDampingOverride=1
    angularSpringConstant=1500
    bAngularSpring=true
    bGravityOverrideInSimSpace=true
    bOverrideAngularDamping=true
    gravityScale=0.35
    simulationSpace=RootRelative
    alphaBoolBlend {blendOption=Linear}
    boundBone {boneName=calfFlap_2_l}
    physicsBodyDefinitions [{
      sphereCollisionRadius=0
      boundBone {boneName=calfFlap_2_l}
      boxExtents=<10 20 20>
      constraintSetup {angularLimitsMax=<0 0 0>, angularLimitsMin=<0 0 -40>, angularTarget=<0 0 0>}
      localJointOffset=<10 0 0>
     }]
   }
   n27 = "Transform (Modify) Bone - Bone: backTubeAttach_c"(ComponentPose=@n26, Alpha=Get head_rot_back(), Translation="-4,-3,0", Rotation="0,0,0", Scale="1,1,1") {
    rotationMode=BMM_Additive
    rotationSpace=BCS_BoneSpace
    scaleMode=BMM_Additive
    scaleSpace=BCS_BoneSpace
    translationMode=BMM_Additive
    translationSpace=BCS_ParentBoneSpace
    alphaBoolBlend {blendOption=Linear}
    alphaScaleBiasClamp {bInterpResult=true, interpSpeedDecreasing=25, interpSpeedIncreasing=25}
    boneToModify {boneName=backTubeAttach_c}
    translation=<-4 -3>
   }
   n28 = Get head_rot_back()
   n29 = "Transform (Modify) Bone - Bone: head_collar_squish_r"(ComponentPose=@n27, Alpha=@n28, Translation="-4.5,-8,7", Rotation="0,0,0", Scale="1,1,1") {
    rotationMode=BMM_Additive
    rotationSpace=BCS_BoneSpace
    scaleMode=BMM_Additive
    scaleSpace=BCS_BoneSpace
    translationMode=BMM_Additive
    translationSpace=BCS_ParentBoneSpace
    alphaBoolBlend {blendOption=Linear}
    alphaScaleBiasClamp {interpSpeedDecreasing=25, interpSpeedIncreasing=25}
    boneToModify {boneName=head_collar_squish_r}
    translation=<-4.5 -8 7>
   }
   n30 = "Transform (Modify) Bone - Bone: head_collar_squish_l"(ComponentPose=@n29, Alpha=@n28, Translation="-4.5,-8,-7", Rotation="0,0,0", Scale="1,1,1") {
    rotationMode=BMM_Additive
    rotationSpace=BCS_BoneSpace
    scaleMode=BMM_Additive
    scaleSpace=BCS_BoneSpace
    translationMode=BMM_Additive
    translationSpace=BCS_ParentBoneSpace
    alphaBoolBlend {blendOption=Linear}
    alphaScaleBiasClamp {interpSpeedDecreasing=25, interpSpeedIncreasing=25}
    boneToModify {boneName=head_collar_squish_l}
    translation=<-4.5 -8 -7>
   }
   n31 = "Transform (Modify) Bone - Bone: neckCollar_l"(ComponentPose=@n30, Alpha=Get head_rot_left(), Translation="0,0,-5", Rotation="0,0,0", Scale="1,1,1") {
    rotationMode=BMM_Additive
    rotationSpace=BCS_BoneSpace
    scaleMode=BMM_Additive
    scaleSpace=BCS_BoneSpace
    translationMode=BMM_Additive
    translationSpace=BCS_ParentBoneSpace
    alphaBoolBlend {blendOption=Linear}
    alphaScaleBiasClamp {interpSpeedDecreasing=25, interpSpeedIncreasing=25}
    boneToModify {boneName=neckCollar_l}
    translation {z=-5}
   }
   n32 = Get ads_rotate()
   n33 = "Transform (Modify) Bone - Bone: neckCollar_r"(ComponentPose=@n31, Alpha=@n32, Translation="-7,0,3.5", Rotation="0,0,0", Scale="1,1,1") {
    rotationMode=BMM_Additive
    rotationSpace=BCS_BoneSpace
    scaleMode=BMM_Additive
    scaleSpace=BCS_BoneSpace
    translationMode=BMM_Additive
    translationSpace=BCS_ParentBoneSpace
    alphaBoolBlend {blendOption=Linear}
    alphaScaleBiasClamp {bInterpResult=true, interpSpeedDecreasing=25, interpSpeedIncreasing=25}
    boneToModify {boneName=neckCollar_r}
    translation {x=-7, z=3.5}
   }
   n34 = "Transform (Modify) Bone - Bone: clavicle_ads_r"(ComponentPose=@n33, Alpha=@n32, Translation="-3.5,0,0", Rotation="0,0,0", Scale="1,1,1") {
    rotationMode=BMM_Additive
    rotationSpace=BCS_BoneSpace
    scaleMode=BMM_Additive
    scaleSpace=BCS_BoneSpace
    translationMode=BMM_Additive
    translationSpace=BCS_ParentBoneSpace
    alphaBoolBlend {blendOption=Linear}
    alphaScaleBiasClamp {bInterpResult=true, interpSpeedDecreasing=25, interpSpeedIncreasing=25}
    boneToModify {boneName=clavicle_ads_r}
    translation {x=-3.5}
   }
   n35 = Get thigh_rot_r()
   n36 = "Transform (Modify) Bone - Bone: pelvis_r"(ComponentPose=@n34, Alpha=@n35, Translation="0,0,0", Rotation="-3.6,-46.55,22.22", Scale="1,1,1") {
    rotationMode=BMM_Additive
    rotationSpace=BCS_BoneSpace
    scaleMode=BMM_Additive
    scaleSpace=BCS_BoneSpace
    translationMode=BMM_Additive
    translationSpace=BCS_ParentBoneSpace
    alphaBoolBlend {blendOption=Linear}
    alphaScaleBiasClamp {interpSpeedDecreasing=25, interpSpeedIncreasing=25}
    boneToModify {boneName=pelvis_r}
    rotation=<rot -3.6 -46.55 22.22>
   }
   n37 = "Transform (Modify) Bone - Bone: thigh_slide_01_r"(ComponentPose=@n36, Alpha=@n35, Translation="7.97,-1.15,-0.36", Rotation="0,0,0", Scale="1,1,1") {
    rotationMode=BMM_Additive
    rotationSpace=BCS_BoneSpace
    scaleMode=BMM_Additive
    scaleSpace=BCS_BoneSpace
    translationMode=BMM_Additive
    translationSpace=BCS_ParentBoneSpace
    alphaBoolBlend {blendOption=Linear}
    alphaScaleBiasClamp {interpSpeedDecreasing=25, interpSpeedIncreasing=25}
    boneToModify {boneName=thigh_slide_01_r}
    translation=<7.97 -1.15 -0.36>
   }
   n38 = "Transform (Modify) Bone - Bone: thigh_front_r"(ComponentPose=@n37, Alpha=@n35, Translation="0,8.46,0", Rotation="0,0,0", Scale="1,1,1") {
    rotationMode=BMM_Additive
    rotationSpace=BCS_BoneSpace
    scaleMode=BMM_Additive
    scaleSpace=BCS_BoneSpace
    translationMode=BMM_Additive
    translationSpace=BCS_ParentBoneSpace
    alphaBoolBlend {blendOption=Linear}
    alphaScaleBiasClamp {interpSpeedDecreasing=25, interpSpeedIncreasing=25}
    boneToModify {boneName=thigh_front_r}
    translation {y=8.46}
   }
   n39 = "Transform (Modify) Bone - Bone: thigh_volume_r"(ComponentPose=@n38, Alpha=@n35, Translation="0,3,0", Rotation="0,0,0", Scale="1,1,1") {
    rotationMode=BMM_Additive
    rotationSpace=BCS_BoneSpace
    scaleMode=BMM_Additive
    scaleSpace=BCS_BoneSpace
    translationMode=BMM_Additive
    translationSpace=BCS_ParentBoneSpace
    alphaBoolBlend {blendOption=Linear}
    alphaScaleBiasClamp {interpSpeedDecreasing=25, interpSpeedIncreasing=25, scale=0}
    boneToModify {boneName=thigh_volume_r}
    translation {y=3}
   }
   n40 = "Constraint - thighRing_r_zero (PPPPP)"!(
    ComponentPose=@n39
    Alpha=1
    ConstraintWeights_0=0.0891
    ConstraintWeights_1:float=0.2496
    ConstraintWeights_2:float=0.521
    ConstraintWeights_3:float=0.0708
    ConstraintWeights_4:float=0.0695
   ) -> (Pose) {
    alphaBoolBlend {blendOption=Linear}
    boneToModify {boneName=thighRing_r_zero}
    constraintSetup [
     {targetBone {boneName=spine_01}}
     {targetBone {boneName=thigh_slide_01_r}}
     {targetBone {boneName=pelvis_r}}
     {targetBone {boneName=thigh_twist_01_r}}
     {targetBone {boneName=pelvis}}
    ]
    constraintWeights [1,1,1,1,1]
   }
   n41 = "Constraint - thighRing_l_zero (PPP)"!(ComponentPose=@n40, Alpha=1, ConstraintWeights_0=0.0923, ConstraintWeights_1:float=0.3328, ConstraintWeights_2:float=0.5749) -> (Pose) {
    alphaBoolBlend {blendOption=Linear}
    boneToModify {boneName=thighRing_l_zero}
    constraintSetup [{targetBone {boneName=spine_01}}, {targetBone {boneName=thigh_twist_01_l}}, {targetBone {boneName=pelvis}}]
    constraintWeights [1,1,1]
   }
   n42 = Look At - Bone: thighRing_r(ComponentPose=@n41, Alpha=1) {
    alphaBoolBlend {blendOption=Linear}
    boneToModify {boneName=thighRing_r}
    lookAtLocation {x=0}
    lookAtTarget {boneReference {boneName=thighRing_r_tgt}}
    lookAt_Axis {axis=<1 0>}
    lookUp_Axis {axis=<0 1>}
   }
   n43 = Look At - Bone: thighRing_l(ComponentPose=@n42, Alpha=1) {
    alphaBoolBlend {blendOption=Linear}
    boneToModify {boneName=thighRing_l}
    lookAtLocation {x=0}
    lookAtTarget {boneReference {boneName=thighRing_l_tgt}}
    lookAt_Axis {axis=<1 0>}
    lookUp_Axis {axis=<0 1>}
   }
   n44 = Anim Dynamics - Bone: thighRing_l(ComponentPose=@n43, Alpha=1, ExternalForce="0,0,0") {
    angularDampingOverride=1
    angularSpringConstant=1500
    bAngularSpring=true
    bGravityOverrideInSimSpace=true
    bOverrideAngularDamping=true
    gravityScale=0.35
    simulationSpace=RootRelative
    alphaBoolBlend {blendOption=Linear}
    boundBone {boneName=thighRing_l}
    physicsBodyDefinitions [{
      sphereCollisionRadius=0
      boundBone {boneName=thighRing_l}
      boxExtents=<10 20 20>
      constraintSetup {angularLimitsMax=<0 0 0>, angularLimitsMin=<0 0 -40>, angularTarget=<0 0 0>}
      localJointOffset=<10 0 0>
     }]
   }
   n45 = Anim Dynamics - Bone: thighRing_r(ComponentPose=@n44, Alpha=1, ExternalForce="0,0,0") {
    angularDampingOverride=1
    angularSpringConstant=1500
    bAngularSpring=true
    bGravityOverrideInSimSpace=true
    bOverrideAngularDamping=true
    gravityScale=0.35
    simulationSpace=RootRelative
    alphaBoolBlend {blendOption=Linear}
    boundBone {boneName=thighRing_r}
    physicsBodyDefinitions [{
      sphereCollisionRadius=0
      boundBone {boneName=thighRing_r}
      boxExtents=<10 20 20>
      constraintSetup {angularLimitsMax=<0 0 0>, angularLimitsMin=<0 0 -40>, angularTarget=<0 0 0>}
      localJointOffset=<10 0 0>
     }]
   }
   n46 = "Constraint - wristFlap_r_zero (PPPPP)"!(
    ComponentPose=@n45
    Alpha=1
    ConstraintWeights_0=0.4696
    ConstraintWeights_1:float=0.0136
    ConstraintWeights_2:float=0.179
    ConstraintWeights_3:float=0.2069
    ConstraintWeights_4:float=0.1308
   ) -> (Pose) {
    alphaBoolBlend {blendOption=Linear}
    boneToModify {boneName=wristFlap_r_zero}
    constraintSetup [
     {targetBone {boneName=lowerarm_twist_02_r}}
     {targetBone {boneName=upperarm_r}}
     {targetBone {boneName=wristFlap_r}}
     {targetBone {boneName=lowerarm_r}}
     {targetBone {boneName=lowerarm_twist_01_r}}
    ]
    constraintWeights [1,1,1,1,1]
   }
   n47 = "Constraint - wristFlap_l_zero (PPPPP)"!(
    ComponentPose=@n46
    Alpha=1
    ConstraintWeights_0=0.2069
    ConstraintWeights_1:float=0.1308
    ConstraintWeights_2:float=0.179
    ConstraintWeights_3:float=0.0136
    ConstraintWeights_4:float=0.4696
   ) -> (Pose) {
    alphaBoolBlend {blendOption=Linear}
    boneToModify {boneName=wristFlap_l_zero}
    constraintSetup [
     {targetBone {boneName=lowerarm_l}}
     {targetBone {boneName=lowerarm_twist_01_l}}
     {targetBone {boneName=wristFlap_l}}
     {targetBone {boneName=upperarm_l}}
     {targetBone {boneName=lowerarm_twist_02_l}}
    ]
    constraintWeights [1,1,1,1,1]
   }
   n48 = Anim Dynamics - Bone: wristFlap_l(ComponentPose=@n47, Alpha=1, ExternalForce="0,0,0") {
    angularDampingOverride=1
    angularSpringConstant=1500
    bAngularSpring=true
    bGravityOverrideInSimSpace=true
    bOverrideAngularDamping=true
    gravityScale=0.35
    simulationSpace=RootRelative
    alphaBoolBlend {blendOption=Linear}
    boundBone {boneName=wristFlap_l}
    physicsBodyDefinitions [{
      sphereCollisionRadius=0
      boundBone {boneName=wristFlap_l}
      boxExtents=<10 20 20>
      constraintSetup {angularLimitsMax=<0 0 0>, angularLimitsMin=<0 0 -40>, angularTarget=<0 0 0>}
      localJointOffset=<10 0 0>
     }]
   }
   n49 = Anim Dynamics - Bone: wristFlap_r(ComponentPose=@n48, Alpha=1, ExternalForce="0,0,0") {
    angularDampingOverride=1
    angularSpringConstant=1500
    bAngularSpring=true
    bGravityOverrideInSimSpace=true
    bOverrideAngularDamping=true
    gravityScale=0.35
    simulationSpace=RootRelative
    alphaBoolBlend {blendOption=Linear}
    boundBone {boneName=wristFlap_r}
    physicsBodyDefinitions [{
      sphereCollisionRadius=0
      boundBone {boneName=wristFlap_r}
      boxExtents=<10 20 20>
      constraintSetup {angularLimitsMax=<0 0 0>, angularLimitsMin=<0 0 -40>, angularTarget=<0 0 0>}
      localJointOffset=<10 0 0>
     }]
   }
   n50 = "Constraint - torsoPouch_2_r_zero (PPPPPPPP)"!(
    ComponentPose=@n49
    Alpha=1
    ConstraintWeights_0=0.0559
    ConstraintWeights_1:float=0.0234
    ConstraintWeights_2:float=0.0805
    ConstraintWeights_3:float=0.2039
    ConstraintWeights_4:float=0.324
    ConstraintWeights_5:float=0.1482
    ConstraintWeights_6:float=0.1426
    ConstraintWeights_7:float=0.0215
   ) -> (Pose) {
    alphaBoolBlend {blendOption=Linear}
    boneToModify {boneName=torsoPouch_2_r_zero}
    constraintSetup [
     {targetBone {boneName=upperarm_twist_01_r}}
     {targetBone {boneName=upperarm_twist_02_r}}
     {targetBone {boneName=spine_01}}
     {targetBone {boneName=spine_02}}
     {targetBone {boneName=spine_03}}
     {targetBone {boneName=spine_04}}
     {targetBone {boneName=clavicle_r}}
     {targetBone {boneName=pelvis}}
    ]
    constraintWeights [1,1,1,1,1,1,1,1]
   }
   n51 = "Constraint - torsoPouch_1_r_zero (PPPPPPPP)"!(
    ComponentPose=@n50
    Alpha=1
    ConstraintWeights_0=0.0674
    ConstraintWeights_1:float=0.0327
    ConstraintWeights_2:float=0.0649
    ConstraintWeights_3:float=0.1634
    ConstraintWeights_4:float=0.2852
    ConstraintWeights_5:float=0.19
    ConstraintWeights_6:float=0.0149
    ConstraintWeights_7:float=0.1815
   ) -> (Pose) {
    alphaBoolBlend {blendOption=Linear}
    boneToModify {boneName=torsoPouch_1_r_zero}
    constraintSetup [
     {targetBone {boneName=upperarm_twist_01_r}}
     {targetBone {boneName=upperarm_twist_02_r}}
     {targetBone {boneName=spine_01}}
     {targetBone {boneName=spine_02}}
     {targetBone {boneName=spine_03}}
     {targetBone {boneName=spine_04}}
     {targetBone {boneName=spine_05}}
     {targetBone {boneName=clavicle_r}}
    ]
    constraintWeights [1,1,1,1,1,1,1,1]
   }
   n52 = "Constraint - torsoPouch_1_l_zero (PPPPPPPP)"!(
    ComponentPose=@n51
    Alpha=1
    ConstraintWeights_0=0.1806
    ConstraintWeights_1:float=0.0204
    ConstraintWeights_2:float=0.0618
    ConstraintWeights_3:float=0.159
    ConstraintWeights_4:float=0.3162
    ConstraintWeights_5:float=0.208
    ConstraintWeights_6:float=0.013
    ConstraintWeights_7:float=0.0411
   ) -> (Pose) {
    alphaBoolBlend {blendOption=Linear}
    boneToModify {boneName=torsoPouch_1_l_zero}
    constraintSetup [
     {targetBone {boneName=clavicle_l}}
     {targetBone {boneName=upperarm_twist_02_l}}
     {targetBone {boneName=spine_01}}
     {targetBone {boneName=spine_02}}
     {targetBone {boneName=spine_03}}
     {targetBone {boneName=spine_04}}
     {targetBone {boneName=spine_05}}
     {targetBone {boneName=upperarm_twist_01_l}}
    ]
    constraintWeights [1,1,1,1,1,1,1,1]
   }
   n53 = "Constraint - torsoPouch_2_l_zero (PPPPPP)"!(
    ComponentPose=@n52
    Alpha=1
    ConstraintWeights_0=0.1459
    ConstraintWeights_1:float=0.0681
    ConstraintWeights_2:float=0.1998
    ConstraintWeights_3:float=0.3802
    ConstraintWeights_4:float=0.1903
    ConstraintWeights_5:float=0.0157
   ) -> (Pose) {
    alphaBoolBlend {blendOption=Linear}
    boneToModify {boneName=torsoPouch_2_l_zero}
    constraintSetup [
     {targetBone {boneName=clavicle_l}}
     {targetBone {boneName=spine_01}}
     {targetBone {boneName=spine_02}}
     {targetBone {boneName=spine_03}}
     {targetBone {boneName=spine_04}}
     {targetBone {boneName=upperarm_twist_01_l}}
    ]
    constraintWeights [1,1,1,1,1,1]
   }
   n54 = Anim Dynamics - Bone: torsoPouch_1_l(ComponentPose=@n53, Alpha=1, ExternalForce="0,0,0") {
    angularDampingOverride=1
    angularSpringConstant=1500
    bAngularSpring=true
    bGravityOverrideInSimSpace=true
    bOverrideAngularDamping=true
    gravityScale=0.35
    simulationSpace=RootRelative
    alphaBoolBlend {blendOption=Linear}
    alphaScaleBiasClamp {scale=0.3}
    boundBone {boneName=torsoPouch_1_l}
    physicsBodyDefinitions [{
      sphereCollisionRadius=0
      boundBone {boneName=torsoPouch_1_l}
      boxExtents=<10 20 20>
      constraintSetup {angularLimitsMax=<0 0 0>, angularLimitsMin=<0 0 -40>, angularTarget=<0 0 0>}
      localJointOffset=<10 0 0>
     }]
   }
   n55 = Anim Dynamics - Bone: torsoPouch_1_r(ComponentPose=@n54, Alpha=1, ExternalForce="0,0,0") {
    angularDampingOverride=1
    angularSpringConstant=1500
    bAngularSpring=true
    bGravityOverrideInSimSpace=true
    bOverrideAngularDamping=true
    gravityScale=0.35
    simulationSpace=RootRelative
    alphaBoolBlend {blendOption=Linear}
    alphaScaleBiasClamp {scale=0.3}
    boundBone {boneName=torsoPouch_1_r}
    physicsBodyDefinitions [{
      sphereCollisionRadius=0
      boundBone {boneName=torsoPouch_1_r}
      boxExtents=<10 20 20>
      constraintSetup {angularLimitsMax=<0 0 0>, angularLimitsMin=<0 0 -40>, angularTarget=<0 0 0>}
      localJointOffset=<10 0 0>
     }]
   }
   n56 = Anim Dynamics - Bone: torsoPouch_2_l(ComponentPose=@n55, Alpha=1, ExternalForce="0,0,0") {
    angularDampingOverride=1
    angularSpringConstant=1500
    bAngularSpring=true
    bGravityOverrideInSimSpace=true
    bOverrideAngularDamping=true
    gravityScale=0.35
    simulationSpace=RootRelative
    alphaBoolBlend {blendOption=Linear}
    alphaScaleBiasClamp {scale=0.3}
    boundBone {boneName=torsoPouch_2_l}
    physicsBodyDefinitions [{
      sphereCollisionRadius=0
      boundBone {boneName=torsoPouch_2_l}
      boxExtents=<10 20 20>
      constraintSetup {angularLimitsMax=<0 0 0>, angularLimitsMin=<0 0 -40>, angularTarget=<0 0 0>}
      localJointOffset=<10 0 0>
     }]
   }
   n57 = Anim Dynamics - Bone: torsoPouch_2_r(ComponentPose=@n56, Alpha=1, ExternalForce="0,0,0") {
    angularDampingOverride=1
    angularSpringConstant=1500
    bAngularSpring=true
    bGravityOverrideInSimSpace=true
    bOverrideAngularDamping=true
    gravityScale=0.35
    simulationSpace=RootRelative
    alphaBoolBlend {blendOption=Linear}
    alphaScaleBiasClamp {scale=0.3}
    boundBone {boneName=torsoPouch_2_r}
    physicsBodyDefinitions [{
      sphereCollisionRadius=0
      boundBone {boneName=torsoPouch_2_r}
      boxExtents=<10 20 20>
      constraintSetup {angularLimitsMax=<0 0 0>, angularLimitsMin=<0 0 -40>, angularTarget=<0 0 0>}
      localJointOffset=<10 0 0>
     }]
   }
   Output Pose(Result=Component To Local(ComponentPose=@n57)) {layerGroup=None}
   Get head_rot_right() -> (head_rot_right:double)
   Get head_rot_right()
  ]}
 EventGraph {nodes [
   Event BlueprintUpdateAnimation()
   n1 = BreakTransform(InTransform=GetDeltaTransformFromRefPose(self=GetOwnerSkeletalMeshComponent(), BoneName=thigh_r, BaseName=pelvis))
   > Set thigh_rot_r(thigh_rot_r=MapRangeClamped(Value=BreakRotator(InRot=@n1.Rotation).Yaw, InRangeA=0, InRangeB=-90))
   n2 = BreakTransform(InTransform=GetDeltaTransformFromRefPose(self=GetOwnerSkeletalMeshComponent(), BoneName=head, BaseName=spine_05))
   n3 = BreakRotator(InRot=@n2.Rotation)
   > Set head_rot_back(head_rot_back:double=MapRangeClamped(Value=@n3.Yaw, InRangeA=0, InRangeB=-50)) -> (Output_Get:double)
   > Set head_rot_left(head_rot_left:double=MapRangeClamped(Value=@n3, InRangeA=0, InRangeB=30)) -> (Output_Get:double)
   > Set head_rot_right(head_rot_right:double=MapRangeClamped(Value=@n3, InRangeA=0, InRangeB=-30)) -> (Output_Get:double)
   > Set head_trans_right(head_trans_right:double=BreakVector(InVec=@n2).Y) -> (Output_Get:double)
   Event OnBeginTargeting() {is=title K2Node_Event}
   > Set ads_rotate(ads_rotate:double=1) -> (Output_Get:double)
   Event OnEndTargeting() {is=title K2Node_Event}
   > Set ads_rotate(ads_rotate=0)
   TryGetPawnOwner()
   MapRangeClamped(Value=0, InRangeA=0, InRangeB=-30)
  ]}
]
variables [
 ads_rotate {type=number}
 head_rot_back {type=number}
 head_rot_left {type=number}
 head_rot_right {type=number}
 head_trans_right {type=number}
 NewVar_0 {type=Rotator}
 NewVar_1 {type=Vector}
 thigh_rot_r {type=number}
]
