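# Animation Blueprint for the padded armor face accessory on the medium male player, keeping its hose and back tube attachment responsive to head movement.
# It copies the parent mesh's pose, then additively offsets backTubeAttach_c and rotates the hose, weighted by head-back and head-left alphas.
# The event graph derives these alphas from head rotation relative to spine_05; head_rot_right is computed but unused.
# A low-weight, spring-constrained AnimDynamics pass adds secondary hose motion.
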AnimBlueprint /Game/Characters/Player/Male/Medium/Bodies/M_MED_Padded_Armor/Meshes/Parts/M_MED_Padded_Armor_FaceAcc_AnimBP
target_skeleton=&/M_MED_Padded_Armor_FaceAcc_Skeleton
graphs [
 AnimGraph {nodes [
   n1 = "Transform (Modify) Bone - Bone: backTubeAttach_c"(
    ComponentPose=Local To Component(LocalPose=Copy Pose From Mesh() {bUseAttachedParent=true})
    Alpha=Get head_rot_back() -> (head_rot_back:double)
    Translation="-4,-3,0"
    Rotation="0,0,0"
    Scale="1,1,1"
   ) {
    rotationMode=BMM_Additive
    rotationSpace=BCS_BoneSpace
    scaleMode=BMM_Additive
    scaleSpace=BCS_BoneSpace
    translationMode=BMM_Additive
    translationSpace=BCS_ParentBoneSpace
    alphaBoolBlend {blendOption=Linear}
    alphaScaleBiasClamp {interpSpeedDecreasing=25, interpSpeedIncreasing=25}
    boneToModify {boneName=backTubeAttach_c}
    translation=<-4 -3>
   }
   n2 = "Transform (Modify) Bone - Bone: hose"(ComponentPose=@n1, Alpha=Get head_rot_left() -> (head_rot_left:double), Translation="0,0,0", Rotation="0,-25,0", Scale="1,1,1") {
    rotationMode=BMM_Additive
    rotationSpace=BCS_BoneSpace
    scaleMode=BMM_Additive
    scaleSpace=BCS_BoneSpace
    translationMode=BMM_Additive
    translationSpace=BCS_ParentBoneSpace
    alphaBoolBlend {blendOption=Linear}
    alphaScaleBiasClamp {interpSpeedDecreasing=25, interpSpeedIncreasing=25}
    boneToModify {boneName=hose}
    rotation {yaw=-25}
   }
   n3 = Anim Dynamics - Bone: hose(ComponentPose=@n2, Alpha=0.2, ExternalForce="0,0,0") {
    angularDampingOverride=1
    angularSpringConstant=1500
    bAngularSpring=true
    bGravityOverrideInSimSpace=true
    bOverrideAngularDamping=true
    gravityScale=0.35
    simulationSpace=RootRelative
    alphaBoolBlend {blendOption=Linear}
    boundBone {boneName=hose}
    physicsBodyDefinitions [{
      sphereCollisionRadius=0
      boundBone {boneName=hose}
      boxExtents=<10 20 20>
      constraintSetup {angularLimitsMax=<0 90 90>, angularLimitsMin=<0 -90 -60>, angularTarget=<1 0 0>}
      localJointOffset=<10 0 0>
     }]
   }
   Output Pose(Result=Component To Local(ComponentPose=@n3)) {layerGroup=None}
  ]}
 EventGraph {nodes [
   Event BlueprintUpdateAnimation()
   n1 = GetOwnerSkeletalMeshComponent()
   n2 = BreakRotator(InRot=BreakTransform(InTransform=GetDeltaTransformFromRefPose(self=@n1, BoneName=head, BaseName=spine_05)).Rotation)
   > Set head_rot_back(head_rot_back:double=MapRangeClamped(Value=@n2.Yaw, InRangeA=0, InRangeB=-50)) -> (Output_Get:double)
   > Set head_rot_left(head_rot_left:double=MapRangeClamped(Value=@n2, InRangeA=0, InRangeB=30)) -> (Output_Get:double)
   > Set head_rot_right(head_rot_right:double=MapRangeClamped(Value=@n2, InRangeA=0, InRangeB=-30)) -> (Output_Get:double)
   TryGetPawnOwner()
  ]}
]
variables [
 head_rot_back {type=number}
 head_rot_left {type=number}
 head_rot_right {type=number}
]
